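# Sound cue for the drop sound of the Warforged AR (SunRose Assault) weapon.
# A Random node picks one of three equally weighted drop waves and feeds a Modulator at default settings for variation.
# It uses the shared weapon drop attenuation and the loot collect sound class.
# The Weapon_Drop_StopOldest concurrency set limits overlapping drops.
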SoundCue /SunRoseWeaponsGameplay/Items/Weapons/WarforgedAR/Sounds/SunRose_Assault_Drop_Cue
root=Modulator_1
nodes [
 {id=Modulator_1, children [Random_2]}
 {id=Random_2, children [WavePlayer_0,WavePlayer_1,WavePlayer_2], properties {weights [1,1,1]}}
 {id=WavePlayer_0, properties {soundWaveAssetPtr=&/SunRose_Assault_Drop_01}}
 {id=WavePlayer_1, properties {soundWaveAssetPtr=&/SunRose_Assault_Drop_02}}
 {id=WavePlayer_2, properties {soundWaveAssetPtr=&/SunRose_Assault_Drop_03}}
]
properties {
 attenuationSettings=&Fort_Weapon_Drop_Att
 maxDistance=3000
 soundClassObject=&Fort_Loot_Collect
 volumeMultiplier=1
 concurrencyOverrides {resolutionRule=PreventNew}
 concurrencySet [&Weapon_Drop_StopOldest]
}
