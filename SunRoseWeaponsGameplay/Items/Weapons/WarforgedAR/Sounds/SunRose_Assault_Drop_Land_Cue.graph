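# Sound cue for the landing sound when the SunRose Warforged assault rifle is dropped, using the loot-collect sound class and weapon-drop attenuation.
# A random node picks one of three equally weighted land waves, which passes through a short delay and then a modulator at default settings.
# Overridden concurrency caps simultaneous instances at a small count and rejects new ones beyond it.
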SoundCue /SunRoseWeaponsGameplay/Items/Weapons/WarforgedAR/Sounds/SunRose_Assault_Drop_Land_Cue
root=Modulator_0
nodes [
 {id=Delay_0, children [Random_0], properties {delayMin=0.1}}
 {id=Modulator_0, children [Delay_0]}
 {id=Random_0, children [WavePlayer_0,WavePlayer_1,WavePlayer_2], properties {weights [1,1,1]}}
 {id=WavePlayer_0, properties {soundWaveAssetPtr=&/SunRose_Assault_Drop_Land_01}}
 {id=WavePlayer_1, properties {soundWaveAssetPtr=&/SunRose_Assault_Drop_Land_02}}
 {id=WavePlayer_2, properties {soundWaveAssetPtr=&/SunRose_Assault_Drop_Land_03}}
]
properties {
 attenuationSettings=&Fort_Weapon_Drop_Att
 bOverrideConcurrency=true
 maxDistance=3000
 soundClassObject=&Fort_Loot_Collect
 volumeMultiplier=1.2
 attenuationOverrides {falloffDistance=1000, attenuationShapeExtents {x=0}}
 concurrencyOverrides {maxCount=3, resolutionRule=PreventNew}
}
