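# Sound cue for the equip foley of the SunRose Assault weapon (WarforgedAR).
# A Random node picks one of three equally weighted equip waves, and a Modulator applies slight pitch and volume variation.
# The cue plays in the weapon foley sound class with weapon reload attenuation, at roughly half volume.
# It overrides concurrency and has the debug flag left on.
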SoundCue /SunRoseWeaponsGameplay/Items/Weapons/WarforgedAR/Sounds/SunRose_Assault_Equip_Cue
root=Modulator_0
nodes [
 {id=Modulator_0, children [Random_1], properties {pitchMax=1.01, pitchMin=0.98, volumeMin=0.98}}
 {id=Random_1, children [WavePlayer_0,WavePlayer_1,WavePlayer_2], properties {weights [1,1,1]}}
 {id=WavePlayer_0, properties {soundWaveAssetPtr=&/SunRose_Assault_Equip_01}}
 {id=WavePlayer_1, properties {soundWaveAssetPtr=&/SunRose_Assault_Equip_02}}
 {id=WavePlayer_2, properties {soundWaveAssetPtr=&/SunRose_Assault_Equip_03}}
]
properties {
 attenuationSettings=&Fort_Weapon_Reload_Att
 bDebug=true
 bOverrideConcurrency=true
 maxDistance=2250
 soundClassObject=&Fort_WeaponFoley
 volumeMultiplier=0.501
}
