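# Sound cue for the drum magazine insert during the SunRose Warforged AR assault rifle reload.
# A mixer layers a random pick of three insert sounds with a slightly quieter, random pick of two drum impact sounds, held back by a fixed one-second delay.
# It uses the weapon foley sound class and shared reload attenuation, overrides concurrency, and has its debug flag switched on.
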SoundCue /SunRoseWeaponsGameplay/Items/Weapons/WarforgedAR/Sounds/SunRose_Assault_Reload_Drum_Insert_Cue
root=Mixer_0
nodes [
 {id=Delay_0, children [Random_3], properties {delayMax=1, delayMin=1}}
 {id=Mixer_0, children [Random_1,Delay_0], properties {inputVolume [1,0.7]}}
 {id=Random_1, children [WavePlayer_3,WavePlayer_4,WavePlayer_5], properties {weights [1,1,1]}}
 {id=Random_3, children [WavePlayer_9, WavePlayer_10], properties {weights [1,1]}}
 {id=WavePlayer_10, properties {soundWaveAssetPtr=&/Sunrose_Assault_Mag_Drum_Impact_01}}
 {id=WavePlayer_3, properties {soundWaveAssetPtr=&/SunRose_Assault_Reload_Insert_01}}
 {id=WavePlayer_4, properties {soundWaveAssetPtr=&/SunRose_Assault_Reload_Insert_02}}
 {id=WavePlayer_5, properties {soundWaveAssetPtr=&/SunRose_Assault_Reload_Insert_03}}
 {id=WavePlayer_9, properties {soundWaveAssetPtr=&/Sunrose_Assault_Mag_Drum_Impact_02}}
]
properties {
 attenuationSettings=&Fort_Weapon_Reload_Att
 bDebug=true
 bOverrideConcurrency=true
 maxDistance=2250
 soundClassObject=&Fort_WeaponFoley
 volumeMultiplier=0.451
}
